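; Fishing line for the FloppingRabbit fishing item, a hook-gun rope that reels caught items back to the player.
; On end-reel it plays montages and sounds while a timeline moves, scales and wiggles the rope meshes.
; When reeling finishes or the owner is lost, DropAndDestroy, gated by Toss, releases attached actors, adds ammo for fished weapons and expires.
; Non-fishing pickups attach to the item socket and fire a server-side quest event once.
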(event Custom|DropAndDestroy
  (bind _outactors (Actor|GetAttachedActors true))
  ;[1 1 1] Drop the items picked caught.
  (if (Variables|Default|GetToss)
    (Utilities|FlowControl|DoOnce false
      (:Completed
        (if (> (Utilities|Array|Length _outactors) 0)
          (for _array_element _outactors
            (Utilities|IsValid _array_element
              (:Is%20Valid
                (bind _asfort_pickup (Utilities|Casting|CastToFortPickup _array_element)
                  (:then
                    (Pickup|TossPickup _asfort_pickup (Transformation|GetActorLocation _asfort_pickup) 0 0 true false "Tossed" "Unset")
                    (if (not (Utilities|Array|ContainsItem (Variables|Default|GetPickedUpPickups) _asfort_pickup))
                      (bind _asfort_weapon_item_definition (Utilities|Casting|CastToFortWeaponItemDefinition (Class|FortItemFunctionLibrary|BreakFortItemEntry (Pickup|GetItemEntry _asfort_pickup 0))))
                      (Fortnite|Weapon|CreateTossAmmoPickupforWeaponItemDefinitionatLocation _asfort_weapon_item_definition (Variables|Default|GetT_FlopperSpawn) (Transformation|GetActorLocation _asfort_pickup))
                      (Utilities|FlowControl|DoOnce false
                        (:Completed
                          (Audio|SpawnSoundAttached (Variables|Default|GetSound_Pickup) (Variables|Transformation|GetRootComponent))))))
                  (:CastFailed
                    ;[1 1 1] Activate caught actor movement component
                    (Utilities|IsValid (Utilities|Array|Get(acopy) (Actor|GetComponentsByClass _array_element "/Script/FortniteGame.FortProjectileMovementComponent"))
                      (:Is%20Valid
                        (CallFunction|ActivateProjectileMovmentComponent (Utilities|Array|Get(acopy) (Actor|GetComponentsByClass _array_element "/Script/FortniteGame.FortProjectileMovementComponent"))))))))))
          (else
            _))
        (Actor|SetLifeSpan 0.1)))
    (else
      _)))

(event EventDestroyed
  ;[1 1 1] Destroyed > Ensure we hide the Hook On destroy
  (Rendering|SetVisibility (Variables|Default|GetHook) false false))

(event Custom|PlayMontage
  (bind _playerpawn (Variables|Default|GetPlayerPawn))
  ;[1 1 1] Play Montage
  (Animation|PlayAnimMontage _playerpawn (Variables|Default|GetReelMontage) 1.0 "None")
  (Animation|Montage|PlayMontage (Variables|/Script/FortniteGame.FortWeapon|GetWeaponMesh (Variables|/Script/FortniteGame.FortPawn|GetCurrentWeapon _playerpawn)) (Variables|Default|GetReelWeaponMontage)))

(event Custom|ActivateProjectileMovmentComponent
  (params (Target :type Actor%20Component%20Object%20Reference))
  ;[1 1 1] Activate caught actor movement component
  (Components|Activation|Activate Target true)
  (Development|PrintString (Utilities|GetDisplayName Target) true true "(R=1.000000,G=0.145246,B=0.000000,A=1.000000)" 5.0))

(event EventBeginPlay
  (EventDispatchers|BindEventtoOnEndReel (delegate Custom|StartReeling)))

(event Custom|StartReeling
  (bind _startlocation (Variables|FortHookGunRope|GetStartLocation))
  (bind _ropemesh (Variables|StaticMesh|GetRopeMesh))
  (bind _projectile_actor (Variables|Default|GetProjectile%20Actor))
  (bind _endlocation (Variables|FortHookGunRope|GetEndLocation))
  ;[1 1 1] Reel In
  (CallFunction|PlayMontage)
  (Utilities|FlowControl|Sequence
    (:then_0
      (Audio|SpawnSoundAttached (Variables|Default|GetReelSound) (Variables|Transformation|GetRootComponent)))
    (:then_1
      (Rendering|SetVisibility (Variables|Default|GetHook) true false))
    (:then_2
      (Utilities|IsValid _projectile_actor
        (:Is%20Valid
          (Collision|SetCollisionEnabled (Variables|/Script/FortniteGame.FortProjectileBase|GetCapsuleComponent _projectile_actor) "NoCollision"))))
    (:then_3
      (Components|Timeline|SetPlayRate (Variables|B_Athena_FloppingRabbit_Wire|GetReel%20In) (Variables|Default|GetReelinPlayRate))
      (bind (_direction _reel_lerp_controller _tighten) (Timeline|Reel%20In PlayFromStart)
        (:Update
          (CallFunction|SetStartPointOfRope)
          (Transformation|SetWorldTransform (Variables|StaticMesh|GetRopeProjectileMesh) (Math|Transform|MakeTransform (Math|Vector|Lerp(Vector) _endlocation _startlocation _reel_lerp_controller) (Math|Rotator|FindLookatRotation _startlocation _endlocation)) false false)
          (bind _distance (Class|MissionBlueprintFunctionLibrary|DistanceBetweenTwoVectors _startlocation (Math|Vector|Lerp(Vector) _endlocation _startlocation _reel_lerp_controller)))
          (Transformation|SetRelativeScale3D (Variables|StaticMesh|GetRopeMesh) (Math|Vector|MakeVector (/ _distance 1e+02) 1.0 1.0))
          (Rendering|Material|SetScalarParameterValueonMaterials _ropemesh "ReturnWiggle" (Math|Float|Lerp 0.0 3.0 _tighten))
          (Rendering|Material|SetScalarParameterValueonMaterials _ropemesh "WiggleAmount" (Math|Float|Lerp 0.0 15.0 _tighten))
          (Transformation|SetWorldRotation (Variables|StaticMesh|GetRopeMesh) (Math|Rotator|FindLookatRotation (Variables|FortHookGunRope|GetStartLocation) (Variables|FortHookGunRope|GetEndLocation)) false false)
          (Utilities|IsValid (Actor|GetOwner)
            (:Is%20Not%20Valid
              (Rendering|SetVisibility (Variables|Default|GetHook) false false)
              (CallFunction|DropandDestroy))))
        (:Finished
          (Ability|SendGameplayEventtoActor (Variables|Default|GetPlayerPawn) (Variables|Default|GetT_Stop)) ; Plays reattach anim
          (Development|PrintString "ReeledIn")
          (Audio|SpawnSoundAttached (Variables|Default|GetReattachSound) (Variables|Default|GetHook))
          (Rendering|SetVisibility (Variables|Default|GetHook) false false)
          (CallFunction|DropandDestroy))))))

(event Custom|TightenRope
  ;[1 1 1] Tighten Rope
  (bind (_direction _newtrack_0) (Timeline|Tight PlayFromStart)
    (:Update
      (Rendering|Material|SetScalarParameterValueonMaterials (Variables|StaticMesh|GetRopeMesh) "WiggleAmount" (Math|Float|Lerp 0.0 15.0 _newtrack_0)))))

(event Custom|AttachNonFishingPickups
  (params (InputPin :type Fort%20Pickup%20Object%20Reference%20%28by%20ref%29))
  (bind _playerpawn (Variables|Default|GetPlayerPawn))
  ;[1 1 1] Attach Non Fishing Items
  (Utilities|Array|Add (Variables|Default|GetPickedUpPickups) InputPin)
  (Transformation|AttachActorToComponent InputPin (Variables|Default|GetItemSocket) "None" "SnapToTarget" "KeepRelative" "KeepWorld" true)
  ;[0.05823 0.5 0.4473 0.5] QUESTS: Fire "Caught Pickup" Event
  (Networking|SwitchHasAuthority
    (:Authority
      (Utilities|IsValid _playerpawn
        (:Is%20Valid
          (Utilities|FlowControl|DoOnce false
            (:Completed
              (Ability|SendGameplayEventtoActor _playerpawn "(TagName=\"Event.Quest.CaughtPickup\")" (Utilities|Struct|MakeGameplayEventData "(TagName=\"Event.Quest.CaughtPickup\")" _playerpawn InputPin 0 0 "()" (GameplayTags|GetOwnedGameplayTags _playerpawn :!class /Script/GameplayTags.BlueprintGameplayTagLibrary) (MarkableActor|GetTargetTagsforQuestEvent InputPin))))))))))

(fn OnRep_CatchParticleOn
  (bind _particlesystem (Variables|Default|GetParticleSystem))
  (if (Variables|Default|GetCatchParticleOn)
    (Components|Activation|Activate _particlesystem true)
    (else
      (Components|Activation|Deactivate _particlesystem))))

(fn ConstructionScript)
