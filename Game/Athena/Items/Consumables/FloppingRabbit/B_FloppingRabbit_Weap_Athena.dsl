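; Fishing rod weapon (FloppingRabbit) that manages the lifecycle of its cast projectile and wire.
; On equip it resets bobber and grip, binds the pawn's holster event and gives the pawn the rod mesh for the tether rope.
; Holster, local unequip and tether detach destroy both actors, making the wire drop anything attached.
; RepNotifies toggle lure bone, trajectory indicator and grip animset, while fish loot tier selection forwards to the projectile's fishing interface.
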(event FortWeapon|EventOnEquippedAndAttached
  ;[1 1 1] Reset hidden bobber and anim on equip
  (|CallParentFunction)
  (Variables|Default|SetHideBobber false)
  (Variables|Default|SetOneHandGrip true)
  (CharacterAnimation|SetWeaponPawnAnimsetOverride "/Game/Animation/Game/MainPlayer/AnimAssets/MeleeOneHand_AnimAsset.MeleeOneHand_AnimAsset")
  (CallFunction|BindHolsterEvents)
  (bind _asfort_player_pawn (Utilities|Casting|CastToFortPlayerPawn (Game|GetInstigator)))
  (Class|FortPlayerPawn|SetTetherRopeRodMesh _asfort_player_pawn (Variables|/Script/Engine.SkinnedMeshComponent|GetSkeletalMesh (Variables|Visual|GetWeaponMesh))))

(event Custom|BindHolsterEvents
  ;[1 1 1] Holster Event binding / unbinding
  (bind _asfort_pawn (Utilities|Casting|CastToFortPawn (Game|GetInstigator)))
  (EventDispatchers|AssignOnHolsteredEvent _asfort_pawn (EventDispatchers|CreateEvent :!event Custom|OnHolstered)))

(event Custom|UnbindHolsterEvents
  ;[1 1 1] Holster Event binding / unbinding
  (bind _asfort_pawn (Utilities|Casting|CastToFortPawn (Game|GetInstigator)))
  (EventDispatchers|UnbindEventfromOnHolsteredEvent _asfort_pawn (EventDispatchers|CreateEvent :!event Custom|OnHolstered)))

(event EventTick
  (params (DeltaSeconds :type Float%20%28single-precision%29))
  (|CallParentFunction DeltaSeconds) ; This node is disabled and will not be called.\nDrag off pins to build functionality.
)

(event EventBeginPlay
  (|CallParentFunction))

(event Custom|ForceKillFishing
  (bind _projectile (Variables|Default|GetProjectile))
  (bind _wire (Variables|Default|GetWire))
  (CharacterAnimation|SetWeaponPawnAnimsetOverride "/Game/Animation/Game/MainPlayer/AnimAssets/MeleeOneHand_AnimAsset.MeleeOneHand_AnimAsset")
  ;[1 1 1] Destroy on Unequip
  (Utilities|IsValid _projectile
    (:Is%20Valid
      (Actor|DestroyActor _projectile))
    (:Is%20Not%20Valid
      _))
  (Utilities|IsValid _wire
    (:Is%20Valid
      ;;[1 1 1] Make sure to drop any pickup or actors attached to wire if reeling in.
      (bind _asb_athena_flopping_rabbit_wire (Utilities|Casting|CastToB_Athena_FloppingRabbit_Wire _wire))
      (Class|BAthenaFloppingRabbitWire|DropandDestroy _asb_athena_flopping_rabbit_wire))))

(event FortWeapon|EventOnLocalUnEquip
  (bind _projectile (Variables|Default|GetProjectile))
  (bind _wire (Variables|Default|GetWire))
  ;[1 1 1] Destroy on Unequip
  (|CallParentFunction)
  (CallFunction|UnbindHolsterEvents)
  (Utilities|IsValid _projectile
    (:Is%20Valid
      (Actor|DestroyActor _projectile))
    (:Is%20Not%20Valid
      _))
  (Utilities|IsValid _wire
    (:Is%20Valid
      ;;[1 1 1] Make sure to drop any pickup or actors attached to wire if reeling in.
      (bind _asb_athena_flopping_rabbit_wire (Utilities|Casting|CastToB_Athena_FloppingRabbit_Wire _wire))
      (Class|BAthenaFloppingRabbitWire|DropandDestroy _asb_athena_flopping_rabbit_wire))))

(event FortWeapon|EventOnTetherDetached
  (bind _projectile (Variables|Default|GetProjectile))
  (bind _wire (Variables|Default|GetWire))
  ;[1 1 1] Destroy on Unequip
  (Utilities|IsValid _projectile
    (:Is%20Valid
      (Actor|DestroyActor _projectile))
    (:Is%20Not%20Valid
      _))
  (Utilities|IsValid _wire
    (:Is%20Valid
      ;;[1 1 1] Make sure to drop any pickup or actors attached to wire if reeling in.
      (bind _asb_athena_flopping_rabbit_wire (Utilities|Casting|CastToB_Athena_FloppingRabbit_Wire _wire))
      (Class|BAthenaFloppingRabbitWire|DropandDestroy _asb_athena_flopping_rabbit_wire))))

(event Custom|OnHolstered
  (bind _projectile (Variables|Default|GetProjectile))
  ;[1 1 1] Force reel-in and cleanup on holster (can cause problems with multiple lines otherwise - see glider redeploy)
  (CallFunction|ForceKillFishing)
  (Utilities|IsValid _projectile
    (:Is%20Valid
      (bind _asprj_athena_flopping_rabbit (Utilities|Casting|CastToPrj_Athena_FloppingRabbit _projectile))
      (Class|PrjAthenaFloppingRabbit|HideandKill _asprj_athena_flopping_rabbit))))

(fn DetermineFishLootTierNameAndSetLootForSpawner
  (params (bIsExplosion :type Boolean))
  (bind _asfort_fishing_interface (Utilities|Casting|CastToFortFishingInterface (Variables|Default|GetProjectile))
    (:then
      (bind _returnvalue (Class|FortFishingInterface|DetermineFishLootTierNameandSetLootforSpawner _asfort_fishing_interface bIsExplosion))
      (return _returnvalue))
    (:CastFailed
      (return "None"))))

(fn OnRep_HideBobber
  (bind _weaponmesh (Variables|Visual|GetWeaponMesh))
  (if (Variables|Default|GetHideBobber)
    (Components|SkinnedMesh|HideBonebyName _weaponmesh "Dyn_Lure" "PBO_None")
    ;[1 1 1] Hide and unhide the trajectory indicator for mobile
    (FortWeapon|SetTrajectoryVisibility false)
    (else
      (Components|SkinnedMesh|UnHideBonebyName _weaponmesh "Dyn_Lure")
      (FortWeapon|SetTrajectoryVisibility true))))

(fn OnRep_OneHandGrip
  (if (Variables|Default|GetOneHandGrip)
    (CharacterAnimation|SetWeaponPawnAnimsetOverride "/Game/Animation/Game/MainPlayer/AnimAssets/MeleeOneHand_AnimAsset.MeleeOneHand_AnimAsset")
    (else
      (CharacterAnimation|SetWeaponPawnAnimsetOverride "/Game/Animation/Game/MainPlayer/AnimAssets/FloppingRabbitCast_AnimAsset.FloppingRabbitCast_AnimAsset") ; Set FloppingRabbitCast Animset
    )))

(fn ConstructionScript
  (|CallParentFunction))
