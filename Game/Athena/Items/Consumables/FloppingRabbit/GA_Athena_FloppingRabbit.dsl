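; Gameplay ability for the FloppingRabbit cast weapon, which either casts a projectile or tries to catch with the one already out.
; With the InWorld tag set, activation has the server call TryCatch and then ends; otherwise it runs the parent logic and sets a two-hand cast animset.
; InWorld and Active tags gate activation, while vehicle and open-seat tag checks end the ability in roofed seats.
; Throw start plays the weapon mesh montage.
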(event EventThrowMontageStarted
  ;[1 1 1] Play Weapon Montage
  (Animation|Montage|PlayMontage (Variables|/Script/FortniteGame.FortWeapon|GetWeaponMesh (Variables|/Script/FortniteGame.FortPawn|GetCurrentWeapon (Variables|ReferencedActors|GetPlayerPawn))) (Variables|Default|GetCastMontageWeapon)))

(event Ability|EventActivateAbility
  ;[1 1 1] If projectile is in the world, try to catch instead of spawning a new one.
  (Ability|Tasks|WaitGameplayTagAdd (Variables|Default|GetInVehicleTag)
    (:then
      (Ability|Tasks|WaitGameplayTagRemove (Variables|Default|GetOpenSeatTag)
        (:then
          (bind _endedability (CallFunction|CheckandEndforRoofedSeat))
          (if (not _endedability)
            (bind _asfort_player_pawn (Utilities|Casting|CastToFortPlayerPawn (Ability|GetAvatarActorfromActorInfo))
              (:then
                (if (GameplayTags|TagContainer|CompareTagAssetInterfacetoOtherTags (Variables|Default|GetT_InWorld) _asfort_player_pawn)
                  (if (Networking|IsServer)
                    (bind _asprj_athena_flopping_rabbit (Utilities|Casting|CastToPrj_Athena_FloppingRabbit (Variables|ReferencedActors|GetSpawnedProjectile))
                      (:then
                        (Class|PrjAthenaFloppingRabbit|TryCatch _asprj_athena_flopping_rabbit))
                      (:CastFailed
                        _))
                    (Ability|EndAbility))
                  (else
                    (|CallParentFunction)
                    (bind _asb_flopping_rabbit_weap_athena (Utilities|Casting|CastToB_FloppingRabbit_Weap_Athena (Variables|/Script/FortniteGame.FortPawn|GetCurrentWeapon _asfort_player_pawn)))
                    (Variables|/Game/Athena/Items/Consumables/FloppingRabbit/B_FloppingRabbit_Weap_Athena.B_FloppingRabbit_Weap_Athena_C|SetOneHandGrip false _asb_flopping_rabbit_weap_athena)
                    (CharacterAnimation|SetWeaponPawnAnimsetOverride _asb_flopping_rabbit_weap_athena "/Game/Animation/Game/MainPlayer/AnimAssets/FloppingRabbitCast_AnimAsset.FloppingRabbitCast_AnimAsset") ; Set TwoHand AnimSet
                  )))
              (:CastFailed
                (Ability|EndAbility)))))
        (:Removed
          _)))
    (:Added
      _))
  (CallFunction|CheckandEndforRoofedSeat))

(fn CheckAndEndForRoofedSeat
  (bind _openseattag (Variables|Default|GetOpenSeatTag))
  (bind _invehicletag (Variables|Default|GetInVehicleTag))
  (bind _asfort_player_pawn (Utilities|Casting|CastToFortPlayerPawn (Ability|GetAvatarActorfromActorInfo)))
  (if (and (not (GameplayTags|TagContainer|CompareTagAssetInterfacetoOtherTags _openseattag _invehicletag _asfort_player_pawn)) (.Case_1%20True (GameplayTags|TagContainer|CompareTagAssetInterfacetoOtherTags _openseattag _invehicletag _asfort_player_pawn)))
    (Ability|EndAbility))
  (return (and (not (GameplayTags|TagContainer|CompareTagAssetInterfacetoOtherTags _openseattag _invehicletag _asfort_player_pawn)) (.Case_1%20True (GameplayTags|TagContainer|CompareTagAssetInterfacetoOtherTags _openseattag _invehicletag _asfort_player_pawn)))))

(fn CanActivateAbility
  (params (ActorInfo :type Gameplay%20Ability%20Actor%20Info%20Structure) (Handle :type Gameplay%20Ability%20Spec%20Handle%20Structure))
  (bind _asfort_player_pawn (Utilities|Casting|CastToFortPlayerPawn (Ability|GetAvatarActorfromActorInfo))
    (:then
      (return 0 (or (and (GameplayTags|TagContainer|CompareTagAssetInterfacetoOtherTags (Variables|Default|GetT_InWorld) (Variables|Default|GetT_Active) _asfort_player_pawn) (.Case_1%20True (GameplayTags|TagContainer|CompareTagAssetInterfacetoOtherTags (Variables|Default|GetT_InWorld) (Variables|Default|GetT_Active) _asfort_player_pawn))) (not (.Case_1%20True (GameplayTags|TagContainer|CompareTagAssetInterfacetoOtherTags (Variables|Default|GetT_InWorld) (Variables|Default|GetT_Active) _asfort_player_pawn))))))
    (:CastFailed
      (return 0 false))))

(fn ShouldOnlyShowTrajectoryOnUse
  (return false))
