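; Event-triggered passive gameplay ability for the FloppingRabbit fishing rod that manages the player's fishing state.
; A gameplay tag switch applies or removes active, in-world and reel-in effects and swaps cast animset, grip and bobber visibility.
; EndFishing force-kills fishing and ends the ability on interrupting states, server-detected seat changes or a failsafe delay after projectile destruction.
; Gliding on mobile also hides the bobber until the glide ends.
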(event Custom|EndFishing
  (bind _weaponrod (Variables|Default|GetWeaponRod))
  ;[1 1 1] End Fishing and the Ability
  (Utilities|IsValid _weaponrod
    (:Is%20Valid
      (Class|BFloppingRabbitWeapAthena|ForceKillFishing _weaponrod)
      (Variables|/Game/Athena/Items/Consumables/FloppingRabbit/B_FloppingRabbit_Weap_Athena.B_FloppingRabbit_Weap_Athena_C|SetOneHandGrip true _weaponrod)
      (Variables|/Game/Athena/Items/Consumables/FloppingRabbit/B_FloppingRabbit_Weap_Athena.B_FloppingRabbit_Weap_Athena_C|SetHideBobber false _weaponrod))
    (:Is%20Not%20Valid
      _))
  (Utilities|FlowControl|Delay 0.1)
  (Ability|EndAbility))

(event Custom|Seat%20Changed
  (params (SeatIndex :type Integer))
  ;[1 1 1] Handle Seat Changing, If a seat is changed then end fishing.
  (CallFunction|EndFishing))

(event Ability|EventOnEndAbility
  (params (bWasCancelled :type Boolean))
  ;[1 1 1] Remove Tags
  (Ability|RemoveGameplayEffectFromOwnerWithGrantedTags (Variables|Default|GetT_Active))
  (Ability|RemoveGameplayEffectFromOwnerWithGrantedTags (Variables|Default|GetT_InWorld))
  (Ability|RemoveGameplayEffectFromOwnerWithGrantedTags (Variables|Default|GetT_ReelIn)))

(event Ability|EventActivateAbilityFromEvent
  (params (EventData :type Gameplay%20Event%20Data%20Structure%20%28by%20ref%29))
  (bind (_eventtag _instigator _target) (Utilities|Struct|BreakGameplayEventData EventData))
  ;[1 1 1] Apply the blocking effects while the ability is active. Remove the effect when the ability is called to stop
  (|CallParentFunction EventData)
  (CallFunction|SaveOutRod)
  (switch Utilities|FlowControl|Switch|SwitchonGameplayTag _eventtag
    (:Default
      _)
    (:Event.Athena.Item.FloppingRabbit.Fire
      (Development|PrintString "RabbitFire")
      (Ability|ApplyGameplayEffectToOwner (Variables|Default|GetGE_Active))
      (Variables|Default|SetProjectile _instigator)
      (Variables|Default|SetWire _target)
      (bind (_currentweapon _castedtorabbitweapon) (CallFunction|SetBobberHidden true))
      (Utilities|IsValid _castedtorabbitweapon
        (:Is%20Valid
          (CharacterAnimation|SetWeaponPawnAnimsetOverride _castedtorabbitweapon "/Game/Animation/Game/MainPlayer/AnimAssets/FloppingRabbitCast_AnimAsset.FloppingRabbitCast_AnimAsset") ; Set FloppingRabbitCast Animset
          (if (Networking|IsServer)
            (Variables|/Game/Athena/Items/Consumables/FloppingRabbit/B_FloppingRabbit_Weap_Athena.B_FloppingRabbit_Weap_Athena_C|SetOneHandGrip false _castedtorabbitweapon)
            ;[1 1 1] Handle Seat Changing, If a seat is changed then end fishing.
            (bind _asfort_player_pawn (Utilities|Casting|CastToFortPlayerPawn (Ability|GetAvatarActorfromActorInfo)))
            (bind _asathena_player_controller (Utilities|Casting|CastToAthena_PlayerController (Fortnite|GetFortPlayerControllerfromActor _asfort_player_pawn)))
            (EventDispatchers|AssignOnVehicleSeatTransitionTargetIndexChange _asathena_player_controller (delegate Custom|Seat%20Changed))))))
    (:Event.Athena.Item.FloppingRabbit.Stop
      (Development|PrintString "RabbitStop")
      (Ability|RemoveGameplayEffectFromOwnerWithGrantedTags (Variables|Default|GetT_Active))
      (Ability|RemoveGameplayEffectFromOwnerWithGrantedTags (Variables|Default|GetT_InWorld))
      (Ability|RemoveGameplayEffectFromOwnerWithGrantedTags (Variables|Default|GetT_ReelIn))
      (bind (_currentweapon_2 _castedtorabbitweapon_2) (CallFunction|SetBobberHidden false))
      (Utilities|IsValid _castedtorabbitweapon_2
        (:Is%20Valid
          (CharacterAnimation|SetWeaponPawnAnimsetOverride _castedtorabbitweapon_2 "/Game/Animation/Game/MainPlayer/AnimAssets/MeleeOneHand_AnimAsset.MeleeOneHand_AnimAsset") ; Reset AnimSet to OneHand
          (if (Networking|IsServer)
            (Variables|/Game/Athena/Items/Consumables/FloppingRabbit/B_FloppingRabbit_Weap_Athena.B_FloppingRabbit_Weap_Athena_C|SetOneHandGrip true _castedtorabbitweapon_2)
            (Ability|EndAbility)))))
    (:Event.Athena.Item.FloppingRabbit.ProjDestroyed
      (Ability|ApplyGameplayEffectToOwner (Variables|Default|GetGE_ReelIn)) ; Apply the reel in GE to put granted tags on the player
      (Utilities|FlowControl|Delay (Variables|Default|GetFailsafeDelay)) ; Make sure ability ends if still active after a duration if the projectile is killed
    )
    (:Event.Athena.Item.FloppingRabbit.InWorld
      (Development|PrintString "RabbitFire")
      (Ability|ApplyGameplayEffectToOwner (Variables|Default|GetGE_InWorld)))
    (:Event.Vehicle.Entered
      _)
    (:Event.Vehicle.Exit
      _)
    (:Quest.Metadata.IsGliding
      ;;[1 1 1] Special Case for Mobile Devices: If the player is Gliding, Hide Bobber(and trajectory) for mobile devices until player is done gliding.
      (if (Fortnite|IsMobilePlatform)
        (CallFunction|SetBobberHidden true)
        (Ability|Tasks|WaitGameplayTagRemove "(TagName=\"Quest.Metadata.IsGliding\")"
          (:Removed
            (CallFunction|SetBobberHidden false)))
        (else
          _))
      (CallFunction|EndFishing))
    (:Gameplay.Action.Player.DanceStunned
      _))
  (CallFunction|EndFishing) ; If the tag that comes in is non of the above, force kill fishing. These are other states that need to be handleds uch as DBNO. Look at the triggers for more info
)

(fn SaveOutRod
  (bind _asfort_player_pawn (Utilities|Casting|CastToFortPlayerPawn (Ability|GetAvatarActorfromActorInfo)))
  (bind _asb_flopping_rabbit_weap_athena (Utilities|Casting|CastToB_FloppingRabbit_Weap_Athena (Variables|/Script/FortniteGame.FortPawn|GetCurrentWeapon _asfort_player_pawn)))
  (Variables|Default|SetWeaponRod _asb_flopping_rabbit_weap_athena))

(fn SetBobberHidden
  (params (HideBobber :type Boolean))
  (bind _asfort_player_pawn (Utilities|Casting|CastToFortPlayerPawn (Ability|GetAvatarActorfromActorInfo)))
  (bind _asb_flopping_rabbit_weap_athena (Utilities|Casting|CastToB_FloppingRabbit_Weap_Athena (Variables|/Script/FortniteGame.FortPawn|GetCurrentWeapon _asfort_player_pawn)))
  (Variables|/Game/Athena/Items/Consumables/FloppingRabbit/B_FloppingRabbit_Weap_Athena.B_FloppingRabbit_Weap_Athena_C|SetHideBobber HideBobber _asb_flopping_rabbit_weap_athena)
  (return (Variables|/Script/FortniteGame.FortPawn|GetCurrentWeapon _asfort_player_pawn) _asb_flopping_rabbit_weap_athena))
