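; Flopping Rabbit consumable ability that pulls the affected avatar toward the instigator who caused the effect, gated by scalable-float toggles.
; CalcLaunchValue computes an arced velocity toward the launcher, capping travel at the configured pull amount.
; Player pawns exit skydiving before launch, with restrictions for teammates, and neutral bots turn hostile to the launcher.
; AI pawns skip ignore-tagged cases and switch to falling so nav-mesh walking doesn't block the launch.
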(event Ability|EventActivateAbility
  (bind _returnvalue (Ability|GetAvatarActorfromActorInfo))
  (bind _returnvalue_2 (Ability|GetGrantedbyEffectContext))
  (bind _returnvalue_3 (Ability|EffectContext|GetEffectCauser _returnvalue_2))
  (bind _instigator (Variables|/Script/Engine.Actor|GetInstigator _returnvalue_3))
  (bind _target (Variables|Default|GetTarget))
  (bind _target_2 (Variables|Default|GetTarget))
  ;[1 1 1] Save Player player out, kick from vehicle, stop gliding.
  (if (Math|Conversions|ToBoolean(Integer) (Math|Float|Floor (ScalableFloat|GetValueatLevel (Variables|Default|GetRow_PullEnabled))))
    (Development|PrintString "Pull")
    (Utilities|IsValid _instigator
      (:Is%20Valid
        (bind _asfort_player_pawn (Utilities|Casting|CastToFortPlayerPawn _instigator)
          (:then
            (Variables|Default|SetLauncher _asfort_player_pawn)
            (bind _asfort_player_pawn_2 (Utilities|Casting|CastToFortPlayerPawn _returnvalue)
              (:then
                (Variables|Default|SetTarget _asfort_player_pawn_2)
                (if (Math|Conversions|ToBoolean(Integer) (Math|Float|Floor (ScalableFloat|GetValueatLevel (Variables|Default|GetRow_CanPullFriends))))
                  _
                  (elif (Team|OnSameTeam (Variables|Default|GetLauncher) (Variables|Default|GetTarget))
                    (Ability|EndAbility)))
                ;[1 1 1] If the target is an ally do not pull if they are skydiving or in a vehicle
                (if (Team|OnSameTeam (Variables|Default|GetLauncher) _target_2)
                  (if (or (Vehicle|IsinVehicle _target_2) (Parachute|IsSkydiving _target_2))
                    (Ability|EndAbility)
                    (else
                      _))
                  (else
                    _))
                (Tags|RemoveNonReplicatedTagtoPawn _target (Variables|Default|GetT_InVehicle))
                (Parachute|EndSkydiving _target)
                ;[1 1 1] Launch Player Pawn
                (bind _outlaunchvelocity (CallFunction|CalcLaunchValue (Variables|Default|GetTarget)))
                ;;[1 1 1] Launch Char
                (bind _launchedpawn (Utilities|LaunchPlayerPawn (Variables|Default|GetLauncher) (Variables|Default|GetTarget) _outlaunchvelocity true true false true true true true))
                (Pawn|Character|ResetFallingHeight _launchedpawn)
                ;[1 1 1] MakeNeutralAIsHostile
                (bind _asfort_athena_aibot_controller (Utilities|Casting|CastToFortAthenaAIBotController (Pawn|GetController (Variables|Default|GetTarget)))
                  (:then
                    (if (Utilities|Enum|Equal(Enum) (Team|GetTeamAffiliation _asfort_athena_aibot_controller (Variables|Default|GetLauncher)) "Neutral")
                      (Affiliation|ChangeAttitudeBetweenActors (Variables|Default|GetTarget) (Variables|Default|GetLauncher) "Hostile")
                      (else
                        _))
                    (Utilities|FlowControl|Delay 0.1)
                    (Ability|EndAbility))
                  (:CastFailed
                    _)))
              (:CastFailed
                (bind _asfort_aipawn (Utilities|Casting|CastToFortAIPawn _returnvalue)
                  (:then
                    (bind _output_get (Variables|Default|SetAITarget _asfort_aipawn))
                    ;[1 1 1] Don't Launch if these tags are on the AI
                    (bind _aitarget (Variables|Default|GetAITarget)
                      (:then
                        (if (GameplayTags|HasAnyMatchingGameplayTags _aitarget (Variables|Default|GetTC_TagsForAIToIgnoreLaunch))
                          (Ability|EndAbility)
                          (else
                            ;[1 1 1] Launc AI Pawns (Like Boars)
                            (bind _aitarget_2 (Variables|Default|GetAITarget)
                              (:then
                                (bind _outlaunchvelocity_2 (CallFunction|CalcLaunchValue _aitarget_2))
                                (Pawn|Components|CharacterMovement|SetMovementMode (Variables|/Script/Engine.Character|GetCharacterMovement _output_get) "MOVE_Falling" 0) ; Set Falling in case we were Nav Mesh Walking.\nThat was stopping the Launch from working.
                                ;;[1 1 1] Launch Char
                                (Character|LaunchCharacter _output_get _outlaunchvelocity_2 true true)
                                (Pawn|Character|ResetFallingHeight _output_get)
                                (Utilities|FlowControl|Delay 0.1)
                                (Ability|EndAbility))
                              (:else)))))
                      (:else)))
                  (:CastFailed
                    _)))))
          (:CastFailed
            _))
        (Ability|EndAbility))
      (:Is%20Not%20Valid
        _))
    (else
      (Ability|EndAbility))))

(fn CalcLaunchValue
  (params (Target :type Actor%20Object%20Reference))
  ;[1 1 1] Calc Vel
  (bind _outlaunchvelocity (Game|SuggestProjectileVelocityCustomArc (Transformation|GetActorLocation Target) (+ (Transformation|GetActorLocation Target) (* (Math|Vector|Normalize (- (Transformation|GetActorLocation Target) (Transformation|GetActorLocation (Variables|Default|GetLauncher)))) (* (select (>= (Math|Vector|VectorLength (- (Transformation|GetActorLocation Target) (Transformation|GetActorLocation (Variables|Default|GetLauncher)))) (ScalableFloat|GetValueatLevel (Variables|Default|GetRow_PullAmount))) (ScalableFloat|GetValueatLevel (Variables|Default|GetRow_PullAmount)) (Math|Vector|VectorLength (- (Transformation|GetActorLocation Target) (Transformation|GetActorLocation (Variables|Default|GetLauncher))))) -1.0))) (Variables|Default|GetLaunchGrav) (Variables|Default|GetLaunchArc))) ; Finds out how much force is needed to put the target 300 units in front of where instigator is facing
  (return _outlaunchvelocity))
