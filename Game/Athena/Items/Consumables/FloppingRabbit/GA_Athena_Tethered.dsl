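; Gameplay ability for the fishing-item (FloppingRabbit) tether that links a player to a Shark.
; On activation it applies the tethered effect and, if enabled, box-overlaps toward the tether pawn to damage structures, playing a smash cue.
; A movement-mode delegate plays a landing cue after sufficiently fast falls.
; On end it removes the effect, unbinds, force-ends fishing and clears any lingering tether left by tag cancellation.
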(event Ability|EventActivateAbility
  (|CallParentFunction)
  (bind (_event_data _fort_player_pawn) (Setup|SetupGenericAbility false false)
    (:Output
      (Variables|Default|SetOwningFortPlayerPawn _fort_player_pawn)
      (bind _returnvalue (Ability|CommitAbility))
      (if _returnvalue
        (Character|AssignMovementModeChangedDelegate (Variables|Default|GetOwningFortPlayerPawn) (EventDispatchers|CreateEvent :!event Custom|MovementModeChangedDelegate_Event_0))
        (bind _returnvalue_2 (Ability|ApplyGameplayEffectToOwner "/Game/Athena/Items/Consumables/FloppingRabbit/GE_Athena_Tethered.GE_Athena_Tethered_C"))
        (Variables|Default|SetGE_TetheredHandle _returnvalue_2)
        (CallFunction|BreakBuildingsatPlayerinDirectionOfTether)))))

(event Ability|EventOnEndAbility
  (params (bWasCancelled :type Boolean))
  (Ability|RemoveGameplayEffectFromOwnerWithHandle (Variables|Default|GetGE_TetheredHandle))
  (Character|UnbindEventfromMovementModeChangedDelegate (Variables|Default|GetOwningFortPlayerPawn) (EventDispatchers|CreateEvent :!event Custom|MovementModeChangedDelegate_Event_0))
  ;[1 1 1] Make sure to clean up the fishing stuff just in case it is still active somehow
  (Ability|SendGameplayEvent (Variables|Default|GetEvent_FishingForceEnd))
  (bind _owningfortplayerpawn (Variables|Default|GetOwningFortPlayerPawn :!pinned true)
    (:then
      (Utilities|IsValid (Class|FortPlayerPawn|GetTetherPawn _owningfortplayerpawn) :!pinned true
        (:Is%20Valid
          (Tether|SetTetherPawn _owningfortplayerpawn 0) ; Clears connection to Shark.
        )) ; Typically when we disconnect normally, TetherPawn is already Not Valid.\n\nIf the Ability was cancelled by tags (EX: Boogie Bomb) this IS still Valid. If we\nclear TetherPawn, it will clean up the connection between Shark and Player correctly.
    )
    (:else)) ; <--- from OnEndAbility
)

(event Custom|MovementModeChangedDelegate_Event_0
  (params (Character :type Character%20Object%20Reference) (PrevMovementMode :type EMovementMode%20Enum :default "MOVE_None") (PreviousCustomMode :type Byte))
  (bind (_returnvalue_x _returnvalue_y _returnvalue_z) (Math|Vector|BreakVector (FortMovement|GetPreviousVelocity (Variables|Default|GetOwningFortPlayerPawn))))
  (bind _charactermovement (Variables|/Script/Engine.Character|GetCharacterMovement Character))
  (bind _movementmode (Variables|/Script/Engine.CharacterMovementComponent|GetMovementMode _charactermovement))
  ;[1 1 1] Landing GC Handling
  (if (and (Utilities|Enum|NotEqual(Enum) _movementmode "MOVE_Falling") (Utilities|Enum|Equal(Enum) PrevMovementMode "MOVE_Falling"))
    (if (< _returnvalue_z (Variables|Default|GetmaxSpeedToPlayGCN))
      (Ability|ExecuteGameplayCueWithParamsOnOwner (Variables|Default|GetGCTagPlayerLand) (Ability|GameplayCue|MakeGameplayCueParameters (Math|Float|SelectFloat 1.0 0.0 (Utilities|Enum|Equal(Enum) _movementmode "MOVE_Walking")))))))

(event Custom|BreakBuildingsAtPlayerInDirectionOfTether
  (bind _owningfortplayerpawn (Variables|Default|GetOwningFortPlayerPawn))
  ;[1 1 1] Destroy Buildings
  (Utilities|IsValid _owningfortplayerpawn
    (:Is%20Valid
      (CallFunction|BreakNearbyStructures (Math|Vector|GetForwardVector (TargetSelection|GetDirectionActortoActor _owningfortplayerpawn (Class|FortPlayerPawn|GetTetherPawn _owningfortplayerpawn))) (* (ScalableFloat|GetValueatLevel (Variables|Default|GetBreakStructuresOnTetherRange)) 0.5)))))

(fn BreakNearbyStructures
  (params (ForwardVector :type Vector) (OffsetForward :type Float%20%28double-precision%29))
  (if (Math|Conversions|ToBoolean(Integer) (Math|Float|Truncate (ScalableFloat|GetValueatLevel (Variables|Default|GetBreakStructuresOnTetherEnabled))))
    (LocalVariables|SetwasAnyStructureBroken false)
    (bind _outactors (Collision|BoxOverlapActors (+ (* ForwardVector OffsetForward) (Math|Vector|MakeVector (.x (Transformation|GetActorLocation (Variables|Default|GetOwningFortPlayerPawn))) (.y (Transformation|GetActorLocation (Variables|Default|GetOwningFortPlayerPawn))) (+ (.z (Transformation|GetActorLocation (Variables|Default|GetOwningFortPlayerPawn))) (* (ScalableFloat|GetValueatLevel (Variables|Default|GetBreakStructuresOnTetherRange)) 0.7)))) (Math|Vector|MakeVector (ScalableFloat|GetValueatLevel (Variables|Default|GetBreakStructuresOnTetherRange)) (ScalableFloat|GetValueatLevel (Variables|Default|GetBreakStructuresOnTetherRange)) (ScalableFloat|GetValueatLevel (Variables|Default|GetBreakStructuresOnTetherRange))) (Utilities|Array|MakeArray "ObjectTypeQuery1") "/Script/FortniteGame.BuildingActor" (Utilities|Array|MakeArray (Variables|Default|GetOwningFortPlayerPawn)) :!class /Script/Engine.KismetSystemLibrary))
    (for _array_element _outactors
      (Utilities|IsValid (Actor|GetComponentbyClass _array_element "/Script/FortniteGame.FortHeldObjectComponent")
        (:Is%20Not%20Valid
          (Utilities|Casting|CastToBuildingProp _array_element
            (:then
              _)
            (:CastFailed
              (Utilities|Casting|CastToFortAthenaSupplyDropBalloon _array_element
                (:CastFailed
                  _))))
          (if (Variables|/Script/Engine.Actor|GetbCanBeDamaged _array_element)
            (Ability|ApplyGameplayEffectToTarget (Ability|TargetData|AbilityTargetDatafromActor _array_element) (Variables|Default|GetStructureDamageGE))
            (LocalVariables|SetwasAnyStructureBroken true)))))
    (if (LocalVariables|GetwasAnyStructureBroken)
      (Ability|ExecuteGameplayCueOnOwner (Variables|Default|GetSmashGC))
      (return)
      (else
        (return)))))
