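; Passive triggered gameplay ability for the Athena tether (in the FloppingRabbit consumable folder) that reacts to tether gameplay events on its owner.
; It switches on the event tag: boost begin and player jump fire gameplay cues, while boost end only ends the ability.
; Tether begin sends a ZiplineExit event to the owning actor, which must run on the client for zipline exit.
; Every branch ends the ability immediately, so each activation is one-shot.
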(event Ability|EventOnEndAbility
  (params (bWasCancelled :type Boolean)))

(event Ability|EventActivateAbilityFromEvent
  (params (EventData :type Gameplay%20Event%20Data%20Structure%20%28by%20ref%29))
  (switch Utilities|FlowControl|Switch|SwitchonGameplayTag (Utilities|Struct|BreakGameplayEventData EventData)
    (:Default
      (Ability|EndAbility))
    (:Athena.Tether.Boost.Begin
      ;[1 1 1] Boost
      (Ability|ExecuteGameplayCueWithParamsOnOwner (Variables|Default|GetGCN_BoostTag) (Ability|GameplayCue|MakeGameplayCueParameters :NormalizedMagnitude 0.0 :RawMagnitude 0.0 :Location "0,0,0" :Normal "0,0,0" :SourceObject self :GameplayEffectLevel 1 :AbilityLevel 1 :bReplicateLocationWhenUsingMinimalRepProxy false))
      (Ability|EndAbility))
    (:Athena.Tether.Boost.End
      (Ability|EndAbility))
    (:Athena.Tether.Jump.Player
      ;[1 1 1] Jump
      (Ability|ExecuteGameplayCueOnOwner (Variables|Default|GetGCN_JumpTag))
      (Ability|EndAbility))
    (:Athena.Tether.Begin
      ;[1 1 1] Start Tether By Event (runs on client required for zipline exit)
      (Ability|SendGameplayEventtoActor (Ability|GetOwningActorfromActorInfo) (Variables|Default|GetZiplineExit) (Utilities|Struct|MakeGameplayEventData (Variables|Default|GetZiplineExit)))
      (Ability|EndAbility))))
