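# Physics asset for the Flopping Rabbit gadget, a fishing-rod consumable, that provides physics bodies for its lure and hook.
# Five capsule bodies form a constraint chain from rod_014 through kin_lure, dyn_lure and dyn_hook_mstr to dyn_hook.
# The rod-to-kin_lure joint swings and twists freely, the two hook joints free only both swings, and the kin_lure-to-dyn_lure joint frees only one swing axis.
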PhysicsAsset /Game/Athena/Items/Consumables/FloppingRabbit/Meshes/Flopping_Rabbit_Gadget_Physics
bodies {
 dyn_hook [{shapeType=Capsule, length=5.4235, radius=3.0489, center=<4.3355 0 0.0384>, rotation=<rot 0 -90.0075 -84.08>}]
 dyn_hook_mstr [{shapeType=Capsule, length=0.505, radius=0.505, center=<1.1338 0 0>, rotation=<rot 0 -90 -89.9986>}]
 dyn_lure [{shapeType=Capsule, length=1.5293, radius=6.3469, center=<7.9392 0.0042 0.0053>, rotation=<rot 0 -90.1781 -89.7764>}]
 kin_lure [{shapeType=Capsule, length=0.505, radius=0.505}]
 rod_014 [{shapeType=Capsule, length=15.4357, radius=2.5287}]
}
constraints [
 {bone1Name=dyn_lure, bone2Name=kin_lure, swing1LimitDegrees=5, swing1Motion=Free, swing2LimitDegrees=90, twistLimitDegrees=2.5}
 {
  bone1Name=dyn_hook_mstr
  bone2Name=dyn_lure
  swing1LimitDegrees=45
  swing1Motion=Free
  swing2LimitDegrees=45
  swing2Motion=Free
  twistLimitDegrees=2.5
 }
 {
  bone1Name=dyn_hook
  bone2Name=dyn_hook_mstr
  swing1LimitDegrees=45
  swing1Motion=Free
  swing2LimitDegrees=45
  swing2Motion=Free
  twistLimitDegrees=2.5
 }
 {
  bone1Name=kin_lure
  bone2Name=rod_014
  swing1LimitDegrees=45
  swing1Motion=Free
  swing2LimitDegrees=45
  swing2Motion=Free
  twistLimitDegrees=45
  twistMotion=Free
 }
]
