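# Animation Blueprint for the Flopping Rabbit gadget prop, a rod that switches between closed, extended and tethered poses.
# A GripState machine feeds a trail controller on the rod bone chain, a montage slot, a root-bone rotation and RigidBody physics.
# Each update copies OneHandGrip from the owning weapon and tunes grip rotation and trail sway.
# A failed weapon cast branches to the tethered setup, which zeroes the trail alpha.
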AnimBlueprint /Game/Athena/Items/Consumables/FloppingRabbit/Meshes/Flopping_Rabbit_Gadget_Skeleton_AnimBlueprint
target_skeleton=&/Flopping_Rabbit_Gadget_Skeleton
graphs [
 AnimGraph {nodes [
   Component To Local()
   AnimGraphNode_StateMachine_0 = GripState() {is=title AnimGraphNode_StateMachine}
   n1 = Trail controller - Bone: rod_014!(
    ComponentPose=Local To Component(LocalPose=@AnimGraphNode_StateMachine_0)
    Alpha=Get Trail_Alpha() -> (Trail_Alpha:double)
    RelaxationSpeedScale:float=Get Trail_RelaxScale() -> (Trail_RelaxScale:double)
   ) -> (Pose) {
    chainBoneAxis=Z
    chainLength=14
    maxDeltaTime=0.0333
    relaxationSpeedScale=4
    rotationLimits [
     {limitMax=<180 5 180>, limitMin=<-180 -5 -180>}
     {limitMax=<180 5 180>, limitMin=<-180 -5 -180>}
     {limitMax=<180 180 180>}
     {limitMax=<180 180 180>}
     {limitMax=<180 180 180>}
     {limitMax=<180 180 180>}
     {limitMax=<180 180 180>}
     {limitMax=<180 180 180>}
     {limitMax=<180 180 180>}
     {limitMax=<180 180 180>}
     {limitMax=<180 180 180>}
     {limitMax=<180 180 180>}
     {limitMax=<180 180 180>}
     {limitMax=<180 180 180>}
    ]
    rotationOffsets [<0 0 0>, <0 0 0>, <0 0 0>, <0 0 0>, <0 0 0>, <0 0 0>, <0 0 0>, <0 0 0>, <0 0 0>, <0 0 0>, <0 0 0>, <0 0 0>, <0 0 0>, <0 0 0>]
    trailBone {boneName=rod_014}
   }
   n2 = Slot 'DefaultSlot'(Source=Component To Local(ComponentPose=@n1))
   n3 = RigidBody(ComponentPose="Transform (Modify) Bone - Bone: b_AR_Root"(
     ComponentPose=Local To Component(LocalPose=@n2)
     Alpha=1
     Translation="0,0,0"
     Rotation=Get GripRotation() -> (GripRotation:Rotator)
     Scale="1,1,1"
    ) {rotationMode=BMM_Replace, alphaBoolBlend {blendOption=Linear}, boneToModify {boneName=b_AR_Root}}) {
    bClampLinearTranslationLimitToRefPose=true
    bOverrideWorldGravity=true
    overridePhysicsAsset=&/Flopping_Rabbit_Gadget_Physics
    alphaBoolBlend {blendOption=Linear}
    componentLinearAccScale=<1 1 1>
    componentLinearVelScale=<1 1 1>
    overrideWorldGravity {z=-980}
   }
   Output Pose(Result=Component To Local(ComponentPose=@n3)) {layerGroup=None}
   AnimGraphNode_SubInput_0 = Input Pose()
  ]}
 GripState {
  entry=Closed
  path=AnimGraph.AnimGraphNode_StateMachine_0.GripState
  ids=id
  nodes [
   AnimStateNode_0 = Closed() {is=title AnimStateNode}
   AnimStateNode_1 = Extended() {is=title AnimStateNode}
   AnimStateNode_2 = Tethered() {is=title AnimStateNode}
   AnimStateTransitionNode_0 = AnimStateTransitionNode() {blendMode=Linear, crossfadeDuration=0.3}
   AnimStateTransitionNode_4 = AnimStateTransitionNode()
   AnimStateTransitionNode_5 = AnimStateTransitionNode()
   AnimStateTransitionNode_6 = AnimStateTransitionNode()
  ]
  links [
   {
    dst=Extended
    rule=AnimGraph.AnimGraphNode_StateMachine_0.GripState.AnimStateTransitionNode_0.Transition
    src=Closed
    via=AnimStateTransitionNode_0
   }
   {
    dst=Tethered
    rule=AnimGraph.AnimGraphNode_StateMachine_0.GripState.AnimStateTransitionNode_5.Transition
    src=Closed
    via=AnimStateTransitionNode_5
   }
   {
    dst=Closed
    rule=AnimGraph.AnimGraphNode_StateMachine_0.GripState.AnimStateTransitionNode_4.AnimationTransitionGraph_0
    src=Extended
    via=AnimStateTransitionNode_4
   }
   {
    dst=Closed
    rule=AnimGraph.AnimGraphNode_StateMachine_0.GripState.AnimStateTransitionNode_6.Transition
    src=Tethered
    via=AnimStateTransitionNode_6
   }
  ]
 }
 Closed {
  path=AnimGraph.AnimGraphNode_StateMachine_0.GripState.AnimStateNode_0.Closed
  nodes [Output Animation Pose(Result=Sequence Player 'FloppingRabbit_Closed_Pose_PROP'() {sequence=/Game/Animation/Game/MainPlayer/Combat/Gadgets/Medium/Male/FloppingRabbit/FloppingRabbit_Closed_Pose_PROP})]
 }
 Extended {
  path=AnimGraph.AnimGraphNode_StateMachine_0.GripState.AnimStateNode_1.Extended
  nodes [Output Animation Pose(Result=Sequence Player 'FloppingRabbit_Base_Pose_PROP'() {sequence=/Game/Animation/Game/MainPlayer/Combat/Gadgets/Medium/Male/FloppingRabbit/FloppingRabbit_Base_Pose_PROP})]
 }
 Tethered {
  path=AnimGraph.AnimGraphNode_StateMachine_0.GripState.AnimStateNode_2.Tethered
  nodes [Output Animation Pose(Result=Sequence Player 'FloppingRabbit_Bendn_Pose'() {sequence=/Game/Animation/Game/MainPlayer/Combat/Gadgets/Medium/Male/FloppingRabbit/FloppingRabbit_Bendn_Pose})]
 }
 Transition {
  path=AnimGraph.AnimGraphNode_StateMachine_0.GripState.AnimStateTransitionNode_0.Transition
  nodes [Result(bCanEnterTransition=NOT Boolean(A=Get OneHandGrip() -> (OneHandGrip:bool)))]
 }
 AnimationTransitionGraph_0 {
  path=AnimGraph.AnimGraphNode_StateMachine_0.GripState.AnimStateTransitionNode_4.AnimationTransitionGraph_0
  nodes [Result(bCanEnterTransition=Get OneHandGrip())]
 }
 Transition {
  path=AnimGraph.AnimGraphNode_StateMachine_0.GripState.AnimStateTransitionNode_5.Transition
  nodes [Result(bCanEnterTransition=Get isTethered() -> (isTethered:bool))]
 }
 Transition {
  path=AnimGraph.AnimGraphNode_StateMachine_0.GripState.AnimStateTransitionNode_6.Transition
  nodes [Result(bCanEnterTransition=NOT Boolean(A=Get isTethered()))]
 }
 EventGraph {nodes [
   Event BlueprintUpdateAnimation()
   > Sequence()
   > n1 = Cast To B_FloppingRabbit_Weap_Athena!(execute, Object=GetOwningActor()) -> (then, CastFailed, AsB Flopping Rabbit Weap Athena:B Flopping Rabbit Weap Athena Object Reference) {is=title K2Node_DynamicCast}
   > Set isTethered(isTethered:bool=false) -> (Output_Get:bool)
   > n2 = Set OneHandGrip(OneHandGrip:bool=Get OneHandGrip!(self:B Flopping Rabbit Weap Athena Object Reference=@n1) -> (OneHandGrip)) -> (Output_Get:bool)
   > n3 = Branch(Condition=@n2)
   > Set GripRotation(GripRotation:Rotator="0,90, 0") -> (Output_Get:Rotator)
   > Set Trail_Alpha(Trail_Alpha:double=0.035) -> (Output_Get:double)
   > Set Trail_RelaxScale(Trail_RelaxScale:double=6) -> (Output_Get:double)
   Set GripRotation(execute=@n3.else, GripRotation="0,90, 0")
   > Set Trail_Alpha(Trail_Alpha=0.175)
   > Set Trail_RelaxScale(Trail_RelaxScale=4.85)
   n4 = Branch(execute=@n1.CastFailed)
   > Set isTethered(isTethered=true)
   > Set GripRotation(GripRotation="0,20, 0")
   > Set Trail_Alpha(Trail_Alpha=0)
   Set isTethered(execute=@n4.else, isTethered=false)
   GetDisplayName(Object=GetWeaponPawnAnimSet!(self:Fort Weapon Object Reference=@n1) -> (ReturnValue:Fort Weapon Anim Set Object Reference))
  ]}
]
variables [
 GripRotation {type=Rotator, default=<rot 0 -90 0>}
 isTethered {type=boolean}
 OneHandGrip {type=boolean, default=true}
 Trail_Alpha {type=number, default=0.2}
 Trail_RelaxScale {type=number, default=3.2}
]
