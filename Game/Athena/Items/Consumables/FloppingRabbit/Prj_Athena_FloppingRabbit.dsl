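; Fishing rod bobber projectile (Flopping Rabbit) that catches items in water, tethers to tetherable pawns and attaches nearby pickups.
; On firing, the server spawns the wire and auto-retracts the bobber when too far from the player.
; Entering water spawns a default or free-fishing spawner, then escalating random dips open a catch window.
; TryCatch spawns the catch through the current spawner or retracts; goal-manager registration exposes the floating bobber to AI.
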(event Custom|ForceKill
  (bind _self self)
  (bind _wire (Variables|Default|GetWire))
  ;[1 1 1] Force Kill
  (Utilities|IsValid _self
    (:Is%20Valid
      (Actor|DestroyActor _self))
    (:Is%20Not%20Valid
      _))
  (Utilities|IsValid _wire
    (:Is%20Valid
      (Actor|DestroyActor _wire))))

(event EventDestroyed
  (bind _wire (Variables|Default|GetWire))
  (bind _fluidsim (Variables|Default|GetFluidSim))
  (bind _spawneddefaultspawner (Variables|InSpawnerCheck|GetSpawnedDefaultSpawner))
  ;[1 1 1] Send event on kill
  (CallFunction|UnRegisterwithGoalManager)
  (bind _playerpawn (Variables|Default|GetPlayerPawn)
    (:then
      (Ability|SendGameplayEventtoActor _playerpawn (Variables|Tags|GetEvent_ProjDestroyed))
      (Utilities|IsValid (Class|FortPlayerPawn|GetTetherPawn _playerpawn)
        (:Is%20Valid
          (if (Variables|Default|GetbTetherCanChangeWeapons)
            (Networking|SwitchHasAuthority
              (:Authority
                (Variables|/Script/FortniteGame.FortPawnComponent_Tether|SetbTetheredWithoutWeapon true (Variables|/Script/FortniteGame.FortPlayerPawn|GetTetherComponent _playerpawn)))
              (:Remote
                _))
            (else
              (Ability|SendGameplayEventtoActor 0 (Variables|Tags|GetEvent_Unhooked) (Utilities|Struct|MakeGameplayEventData (Variables|Tags|GetEvent_Unhooked) _playerpawn)) ; UNHOOKED!\n\nLet the TetherPawn know it has been released.
              (Tether|SetTetherPawn _playerpawn 0))))
        (:Is%20Not%20Valid
          _)))
    (:else
      _))
  (CallFunction|WriteNothingCaughtFishingAnalytic)
  (bind _asbga_athena_flopper_spawn_default (Utilities|Casting|CastToBGA_Athena_FlopperSpawn_Default _spawneddefaultspawner)
    (:then
      (Utilities|IsValid _asbga_athena_flopper_spawn_default
        (:Is%20Valid
          (Actor|SetLifeSpan _spawneddefaultspawner 2.0)
          (if (Variables|Default|GetSpawnedItem) :!label _label
            _))
        (:Is%20Not%20Valid
          _)))
    (:CastFailed
      (bind _asbga_athena_flopper_spawn_world (Utilities|Casting|CastToBGA_Athena_FlopperSpawn_World _spawneddefaultspawner)
        (:then
          (Actor|SetLifeSpan _asbga_athena_flopper_spawn_world 2.0)
          (goto _label))
        (:CastFailed
          _))))
  (Utilities|IsValid _fluidsim
    (:Is%20Valid
      (Class|BPFluidSim01|RemoveProjectileForce _fluidsim (Variables|Default|GetMesh)))
    (:Is%20Not%20Valid
      _))
  ;;[1 1 1] Make sure we set projectile gone in cases where we may be destroying the projectile right away.
  (Utilities|IsValid _wire
    (:Is%20Valid
      (Variables|/Script/FortniteGame.FortHookGunRope|SetbProjectileGone true _wire))
    (:Is%20Not%20Valid
      _))
  (|CallParentFunction))

(event EventBeginPlay
  (bind _playerpawn (Variables|Default|GetPlayerPawn))
  (bind _currentweapon (Variables|/Script/FortniteGame.FortPawn|GetCurrentWeapon _playerpawn))
  ;[1 1 1] Begin Play - Set up water flow interactions, spawn rope, and send fire event to player
  (bind _outactors (Actor|GetAllActorsOfClass "/Game/Athena/Apollo/Water/Blueprints/BP_FluidSim_FN.BP_FluidSim_FN_C"))
  (if (> (Utilities|Array|Length _outactors) 0)
    (Variables|Default|SetFluidSim (Utilities|Array|Get(acopy) _outactors)))
  (CallFunction|GetSheetVals)
  (bind _asfort_player_pawn (Utilities|Casting|CastToFortPlayerPawn (Game|GetInstigator)))
  (bind _output_get (Variables|Default|SetPlayerPawn _asfort_player_pawn))
  (if (Networking|IsServer)
    (bind _returnvalue (Game|SpawnActorfromClass (Variables|Default|GetFloppingRabbitWire) (Math|Transform|MakeTransform (Transformation|GetActorLocation)) "Undefined" "SelectDefaultAtRuntime" _currentweapon self _playerpawn _playerpawn))
    (Variables|Default|SetWire _returnvalue)
    (Ability|SendGameplayEventtoActor _output_get (Variables|Tags|GetEvent_Fire))
    ;;[1 1 1] Set wire and projectile on the weapon. IF fail kill everything
    (bind _asb_flopping_rabbit_weap_athena (Utilities|Casting|CastToB_FloppingRabbit_Weap_Athena _currentweapon)
      (:then
        (Variables|/Game/Athena/Items/Consumables/FloppingRabbit/B_FloppingRabbit_Weap_Athena.B_FloppingRabbit_Weap_Athena_C|SetWire (Variables|Default|GetWire) _asb_flopping_rabbit_weap_athena)
        (Variables|/Game/Athena/Items/Consumables/FloppingRabbit/B_FloppingRabbit_Weap_Athena.B_FloppingRabbit_Weap_Athena_C|SetProjectile self _asb_flopping_rabbit_weap_athena)
        (Class|FortHookGunRope|SetWeaponandProjectile (Variables|Default|GetWire) _asb_flopping_rabbit_weap_athena self)
        ;[1 1 1] Distance Check From Player to Auto reel in
        (Utilities|Time|SetTimerbyEvent (delegate Custom|DistanceCheck) (Variables|WaterAndCatchingFunctionality|GetDistanceCheckInterval) true))
      (:CastFailed
        (CallFunction|ForceKill)))))

(event Custom|DistanceCheck
  (bind _playerpawn (Variables|Default|GetPlayerPawn))
  (bind (_returnvalue_x _returnvalue_y _returnvalue_z) (Math|Vector|BreakVector (Transformation|GetActorLocation _playerpawn)))
  (bind _self self)
  (bind (_returnvalue_x_2 _returnvalue_y_2 _returnvalue_z_2) (Math|Vector|BreakVector (Transformation|GetActorLocation _self)))
  ;[1 1 1] Distance Check From Player to Auto reel in
  (bind _distance (Class|MissionBlueprintFunctionLibrary|DistanceBetweenTwoVectors (Math|Vector|MakeVector _returnvalue_x _returnvalue_y) (Math|Vector|MakeVector _returnvalue_x_2 _returnvalue_y_2)))
  (if (and (or (> _distance (ScalableFloat|GetValueatLevel (Variables|Rows|GetRow_DistanceCheck))) (> (Math|Float|Absolute(Float) (- _returnvalue_z _returnvalue_z_2)) (ScalableFloat|GetValueatLevel (Variables|Rows|GetRow_DistanceCheckZ)))) (not (Class|FortPlayerPawn|IsTethered (Variables|Default|GetPlayerPawn))))
    (CallFunction|HideandKill)))

(event Custom|WriteNothingCaughtFishingAnalytic
  ;[1 1 1] Write Finish Event checking for overlap of world pool
  (CallFunction|HandleNothingCaughtFishingAnalytics))

(event Custom|TryCatch
  (bind _spawner (Variables|InSpawnerCheck|GetSpawner))
  (bind _wire (Variables|Default|GetWire))
  ;[1 1 1] Try Catch - If the player can catch, spawn a "fish" then kill, else just kill
  (Utilities|FlowControl|DoOnce false
    (:Completed
      (if (Variables|WaterAndCatchingFunctionality|GetCanCatch)
        (Utilities|IsValid _spawner
          (:Is%20Valid
            (bind _asbga_athena_flopper_spawn_parent (Utilities|Casting|CastToBGA_Athena_FlopperSpawn_Parent _spawner)
              (:then
                (Utilities|IsValid _wire
                  (:Is%20Valid
                    (CallFunction|ListenForFishingComplete true)
                    (Class|BGAAthenaFlopperSpawnParent|SpawnItem _asbga_athena_flopper_spawn_parent (Game|GetInstigator) true (Variables|/Game/Athena/Items/Consumables/FloppingRabbit/B_Athena_FloppingRabbit_Wire.B_Athena_FloppingRabbit_Wire_C|GetItemSocket _wire) (Transformation|GetActorLocation) (Variables|Default|GetPlayerPawn) 0.1 true false (Variables|/Script/FortniteGame.FortPawn|GetCurrentWeapon (Variables|Default|GetPlayerPawn)) self (Variables|Default|GetFishingRodTag))
                    (Utilities|IsValid (Variables|/Game/Athena/Items/EnvironmentalItems/FlopperSpawn/BGA_Athena_FlopperSpawn_Parent.BGA_Athena_FlopperSpawn_Parent_C|GetSpawnedItem _asbga_athena_flopper_spawn_parent)
                      (:Is%20Not%20Valid
                        _)))
                  (:Is%20Not%20Valid
                    _)))
              (:CastFailed
                _)))
          (:Is%20Not%20Valid
            _)))
      (CallFunction|HideandKill))))

(event Custom|CustomPickupAttach
  (params (FortPickup :type Fort%20Pickup%20Object%20Reference))
  ;[1 1 1] Attach other overlapping pickups
  (if (Networking|IsServer)
    (if (Math|Boolean|ANDBoolean (Utilities|NotEqual(Object) (Actor|GetAttachParentActor FortPickup) self) (<= (Transformation|GetSquaredDistanceTo FortPickup self) (Variables|Default|GetItemAttachMinDist)) (>= (Transformation|GetSquaredDistanceTo (Game|GetInstigator) FortPickup) (Variables|Default|GetItemAttachDistanceRequirement)))
      (Class|BAthenaFloppingRabbitWire|AttachNonFishingPickups (Variables|Default|GetWire) FortPickup))))

(event Custom|HitWater
  (params (FishSpawner :type Boolean))
  (bind _waterinteractioncomponent (Variables|Water|GetWaterInteractionComponent))
  (bind (_waterplanelocation _waterplanenormal _watersurfaceposition _waterdepth _waterbodyidx) (FX|GetLastWaterSurfaceInfo _waterinteractioncomponent :!class /Script/FortniteGame.FortWaterInteractionComponent))
  (bind _fluidsim (Variables|Default|GetFluidSim))
  ;[1 1 1] Hit Water - Play FX and Sound, Play High Tier if in a fish spawner
  (Utilities|FlowControl|DoOnce false
    (:Completed
      (if (not (Networking|IsServer))
        (Utilities|FlowControl|Delay 0.01)
        ;;[1 0.4769 0.447] Niagara fx
        (bind _returnvalue (Niagara|SpawnSystematLocation :SystemTemplate (Variables|FXAndSounds|GetNS_FX_HitWater) :Location _waterplanelocation :Scale (Variables|FXandSounds|GetFXScaleHitWater) :bAutoDestroy true :bAutoActivate true :PoolingMethod "None" :bPreCullCheck true))
        (Effects|Components|ParticleSystem|SetFloatParameter _returnvalue "User.WaterBodyIndex" (Math|Conversions|ToFloat(Integer) _waterbodyidx))
        (Effects|Components|ParticleSystem|SetFloatParameter _returnvalue "User.WaterDepth" _waterdepth)
        (Effects|Components|ParticleSystem|SetColorParameter _returnvalue "User.WaterColor" (Variables|Default|GetWaterBodyPrimaryColor))
        (Audio|PlaySoundatLocation (Variables|FXAndSounds|GetSound_HitWater) (Transformation|GetActorLocation))
        ;[1 1 1] handle water dynamic force
        (Utilities|IsValid _fluidsim
          (:Is%20Valid
            (Class|BPFluidSim01|RegisterProjectileForce _fluidsim (Utilities|Struct|MakeFluidForceDynamic "NewEnumerator0" 32.0 -1.0 (Variables|Default|GetMesh) 0 (Utilities|Struct|MakeFluidForceSocketInfo (Utilities|Map|MakeMap "None" "None"))) (Variables|Default|GetMesh))))))))

(event EventOnExploded
  (params (HitActors :type Array%20of%20Actor%20Object%20References) (HitResults :type Array%20of%20Hit%20Result%20Structures)))

(event Custom|ReelIn
  ;[1 1 1] Try Catch - If the player can catch, spawn a "fish" then kill, else just kill
  (CallFunction|ListenForFishingComplete false)
  (CallFunction|SpawnItemsFX)
  (Variables|Default|SetSpawnedItem true)
  (Variables|WaterAndCatchingFunctionality|SetCanCatch false)
  (if (Networking|IsServer)
    (Variables|/Game/Athena/Items/Consumables/FloppingRabbit/B_Athena_FloppingRabbit_Wire.B_Athena_FloppingRabbit_Wire_C|SetCatchParticleOn true (Variables|Default|GetWire)))
  (CallFunction|HideandKill))

(event Custom|CatchWindow
  ;[1 1 1] Catch Window - Set can Catch to True for a brief moment.
  (Variables|WaterAndCatchingFunctionality|SetCanCatch true)
  (Utilities|FlowControl|RetriggerableDelay (Variables|WaterAndCatchingFunctionality|GetDelay_CatchWindow))
  (Variables|WaterAndCatchingFunctionality|SetCanCatch false))

(event Collision|EventHit
  (params (MyComp :type Primitive%20Component%20Object%20Reference) (Other :type Actor%20Object%20Reference) (OtherComp :type Primitive%20Component%20Object%20Reference) (bSelfMoved :type Boolean) (HitLocation :type Vector :default "0, 0, 0") (HitNormal :type Vector :default "0, 0, 0") (NormalImpulse :type Vector :default "0, 0, 0") (Hit :type Hit%20Result%20Structure%20%28by%20ref%29))
  (bind _playerpawn (Variables|Default|GetPlayerPawn))
  ;[1 1 1] Hit Event - If player hide and kill, if not check if in water, if not in water hide and kill. Also don't retract for tetherable pawns, then they can't get the bobber!
  (|CallParentFunction MyComp Other OtherComp bSelfMoved HitLocation HitNormal NormalImpulse Hit)
  (bind _btethered (CallFunction|AttemptTether Other))
  (if _btethered
    (CallFunction|ForceKill)
    (else
      (bind _asfort_pawn (Utilities|Casting|CastToFortPawn Other)
        (:then
          (Utilities|FlowControl|Sequence
            (:then_0
              (if (not (GameplayTags|HasTag (GameplayTags|GetOwnedGameplayTags _asfort_pawn :!class /Script/GameplayTags.GameplayTagAssetInterface) "(TagName=\"NPC.Tetherable\")"))
                (Collision|SetActorEnableCollision false)
                (Delivery|DoExplosion)
                (CallFunction|HideandKill)))
            (:then_1
              ;;[0.05823 0.5 0.4473 0.5] QUESTS: Fire "Caught Pawn" Event
              (Networking|SwitchHasAuthority
                (:Authority
                  (Utilities|IsValid _playerpawn
                    (:Is%20Valid
                      (Utilities|FlowControl|DoOnce false
                        (:Completed
                          (Ability|SendGameplayEventtoActor _playerpawn "(TagName=\"Event.Quest.CaughtPawn\")" (Utilities|Struct|MakeGameplayEventData "(TagName=\"Event.Quest.CaughtPawn\")" _playerpawn _asfort_pawn 0 0 "()" (GameplayTags|GetOwnedGameplayTags _playerpawn :!class /Script/GameplayTags.BlueprintGameplayTagLibrary) (GameplayTags|GetOwnedGameplayTags _asfort_pawn :!class /Script/GameplayTags.BlueprintGameplayTagLibrary))))))))))))
        (:CastFailed
          (Networking|SwitchHasAuthority
            (:Authority
              (if (not (Variables|WaterAndCatchingFunctionality|GetInWater))
                (Delivery|DoExplosion)
                (CallFunction|HideandKill)))))))))

(event Custom|ToggleMiniGameSplashLoop
  (params (On :type Boolean))
  (bind _returnvalue (Ability|GameplayCue|MakeGameplayCueParameters 0.0))
  (bind _self self)
  (bind _gc_minigamesplash (Variables|FXandSounds|GetGC_MiniGameSplash))
  (if On
    (Projectile|AddGameplayCue%28Looping%29Local _self _gc_minigamesplash _returnvalue)
    (else
      (Projectile|RemoveGameplayCue%28Looping%29Local _self _gc_minigamesplash _returnvalue))))

(event Custom|SpawnItemsFX
  (bind _waterinteractioncomponent (Variables|Water|GetWaterInteractionComponent))
  (bind (_waterplanelocation _waterplanenormal _watersurfaceposition _waterdepth _waterbodyidx) (FX|GetLastWaterSurfaceInfo _waterinteractioncomponent :!class /Script/FortniteGame.FortWaterInteractionComponent))
  ;[1 1 1] When Items Spawn Play FX and Sounds
  (Utilities|FlowControl|DoOnce false
    (:Completed
      (if (not (Networking|IsServer))
        ;;[1 0.4769 0.447] Niagara fx
        (bind _returnvalue (Niagara|SpawnSystematLocation (Variables|FXAndSounds|GetNS_FX_SpawnItems) _waterplanelocation))
        (Effects|Components|ParticleSystem|SetFloatParameter _returnvalue "User.WaterBodyIndex" (Math|Conversions|ToFloat(Integer) _waterbodyidx))
        (Effects|Components|ParticleSystem|SetFloatParameter _returnvalue "User.WaterDepth" _waterdepth)
        (Effects|Components|ParticleSystem|SetColorParameter _returnvalue "User.WaterColor" (Variables|Default|GetWaterBodyPrimaryColor))
        (Audio|PlaySoundatLocation (Variables|FXAndSounds|GetSound_SpawnItems) (Transformation|GetActorLocation))))))

(event Custom|DipFx
  ;[1 1 1] Full Dip - Do a big dip into the water and enable the catch window. On Finish, reset catch chance and start the Idle again.
  ;;[1 0.4769 0.447] Niagara fx
  (bind _returnvalue (Niagara|SpawnSystematLocation (select (stale Custom|FullDip InWorldSpawner) (Variables|FXAndSounds|GetNS_FX_WaterDip_HighTier) (Variables|FXAndSounds|GetNS_FX_WaterDip)) (Transformation|GetActorLocation) (Transformation|GetActorRotation) (Variables|FXAndSounds|GetFXScaleWaterDip)))
  (Niagara|SetNiagaraVariable(LinearColor) _returnvalue "WaterColor" (Variables|Default|GetWaterBodyPrimaryColor))
  (Variables|WaterAndCatchingFunctionality|SetMeshLocAfterHit (Variables|/Script/Engine.SceneComponent|GetRelativeLocation (Variables|Default|GetMesh)))
  (Audio|PlaySoundatLocation (select (stale Custom|FullDip InWorldSpawner) (Variables|FXAndSounds|GetSound_WaterDip_HighTier) (Variables|FXAndSounds|GetSound_WaterDip)) (Transformation|GetActorLocation))
  (Fortnite|PlayLocalForceFeedbackatLocation "/Game/Weapons/ForceFeedback/FF_Athena_FloppingRabbit_Dip.FF_Athena_FloppingRabbit_Dip" (Transformation|GetActorLocation (Game|GetInstigator)) 2e+01)
  (bind (_direction _bigdip) (Timeline|BigBob PlayFromStart)
    (:Update
      (Transformation|SetRelativeLocation (Variables|Default|GetMesh) (Math|Vector|vector+vector (Variables|WaterAndCatchingFunctionality|GetMeshLocAfterHit) (Math|Vector|MakeVector 0.0 0.0 _bigdip)) false false))
    (:Finished
      (if (Networking|IsServer)
        (Variables|WaterAndCatchingFunctionality|SetChanceForBigBob (Variables|WaterAndCatchingFunctionality|GetStartingCatchBigBobChance))
        (Variables|Default|SetDipping false)
        (CallFunction|CheckforDip)))))

(event Custom|FullDip
  (params (InWorldSpawner :type Boolean))
  ;[1 1 1] Full Dip - Do a big dip into the water and enable the catch window. On Finish, reset catch chance and start the Idle again.
  (Development|PrintString "FullDip")
  (CallFunction|CatchWindow)
  ;;[1 0.4769 0.447] Niagara fx
  (bind _returnvalue (Niagara|SpawnSystematLocation (select InWorldSpawner (Variables|FXAndSounds|GetNS_FX_WaterDip_HighTier) (Variables|FXAndSounds|GetNS_FX_WaterDip)) (Transformation|GetActorLocation) (Transformation|GetActorRotation) (Variables|FXAndSounds|GetFXScaleWaterDip)))
  (Niagara|SetNiagaraVariable(LinearColor) _returnvalue "WaterColor" (Variables|Default|GetWaterBodyPrimaryColor))
  (Variables|WaterAndCatchingFunctionality|SetMeshLocAfterHit (Variables|/Script/Engine.SceneComponent|GetRelativeLocation (Variables|Default|GetMesh)))
  (Audio|PlaySoundatLocation (select InWorldSpawner (Variables|FXAndSounds|GetSound_WaterDip_HighTier) (Variables|FXAndSounds|GetSound_WaterDip)) (Transformation|GetActorLocation))
  (Fortnite|PlayLocalForceFeedbackatLocation "/Game/Weapons/ForceFeedback/FF_Athena_FloppingRabbit_Dip.FF_Athena_FloppingRabbit_Dip" (Transformation|GetActorLocation (Game|GetInstigator)) 2e+01)
  (Timeline|BigBob PlayFromStart))

(event Custom|HideAndKill
  (bind _wire (Variables|Default|GetWire))
  (bind _spawneddefaultspawner (Variables|InSpawnerCheck|GetSpawnedDefaultSpawner))
  ;[1 1 1] Hide and Kill - Turn of collision, hide, and tighen the rope / reel it in.
  (Utilities|FlowControl|DoOnce false
    (:Completed
      (Development|SetHiddeninGame (Variables|Default|GetMesh) true false)
      (Collision|SetActorEnableCollision false)
      (Utilities|IsValid _wire
        (:Is%20Valid
          (Variables|/Script/FortniteGame.FortHookGunRope|SetbProjectileGone true _wire)
          (Class|BAthenaFloppingRabbitWire|TightenRope _wire))
        (:Is%20Not%20Valid
          _))
      (Utilities|IsValid _spawneddefaultspawner
        (:Is%20Valid
          (Actor|DestroyActor _spawneddefaultspawner))
        (:Is%20Not%20Valid
          _))
      (Actor|SetLifeSpan 0.1)
      (Variables|Default|SetbDidHitMeatball false)
      (CallFunction|UnRegisterwithGoalManager))))

(event Custom|CheckForHighTierSpawner
  ;[1 1 1] Check for high tier spawner
  (bind (_outactors _returnvalue) (Collision|CapsuleOverlapActors (Transformation|GetActorLocation) 7e+01 1e+03 (Variables|InSpawnerCheck|GetObjectTypes) "/Game/Athena/Items/EnvironmentalItems/FlopperSpawn/BGA_Athena_FlopperSpawn_World.BGA_Athena_FlopperSpawn_World_C"))
  (Variables|InSpawnerCheck|SetWorldSpawner _returnvalue)
  (if _returnvalue
    (Utilities|IsValid (Utilities|Array|Get(acopy) _outactors)
      (:Is%20Valid
        (Variables|InSpawnerCheck|SetSpawner (Utilities|Array|Get(acopy) _outactors))
        (Development|PrintString "HighTierFlopping"))
      (:Is%20Not%20Valid
        _)))
  (Variables|InSpawnerCheck|SetSpawner (Variables|InSpawnerCheck|GetSpawnedDefaultSpawner)))

(event Custom|CheckForDip
  ;[1 1 1] Check For Dip
  (Utilities|FlowControl|Delay (Math|Random|RandomFloatinRange (ScalableFloat|GetValueatLevel (Variables|Rows|GetRow_BobChanceDelayMin)) (ScalableFloat|GetValueatLevel (Variables|Rows|GetRow_BobChanceDelayMax))))
  (if (Variables|Default|GetDipping)
    (CallFunction|CheckforDip)
    (elif (Networking|IsServer)
      (if (Math|Random|RandomBoolwithWeight (Variables|WaterAndCatchingFunctionality|GetChanceForBigBob))
        (Variables|Default|SetDipping true)
        (CallFunction|CheckforHighTierSpawner)
        (CallFunction|FullDip (Variables|InSpawnerCheck|GetWorldSpawner))
        (else
          (Variables|WaterAndCatchingFunctionality|SetChanceForBigBob (+ (Variables|WaterAndCatchingFunctionality|GetChanceForBigBob) (Variables|WaterAndCatchingFunctionality|GetIncreaseBigBobChanceAmount)))
          (CallFunction|CheckforDip))))))

(event OnExitWaterEvent(WaterInteractionComponent)
  (params (WaterBody :type Fort%20Water%20Body%20Actor%20Object%20Reference) (WaterInteractionComponent :type Fort%20Water%20Interaction%20Component%20Object%20Reference) (bIsLastBody :type Boolean))
  ;[1 1 1] If we leave water reel in.
  (if (Class|FortPlayerPawn|IsTethered (Variables|Default|GetPlayerPawn))
    _
    (elif (Networking|IsServer)
      (if bIsLastBody
        (Utilities|FlowControl|RetriggerableDelay (Variables|Default|GetBounceOutOfWaterDelay))
        (if (not (Water|IsinWater (Variables|Water|GetWaterInteractionComponent)))
          (CallFunction|HideandKill))))))

(event Custom|ServerAttemptTether
  (params (TetherActor :type Pawn%20Object%20Reference))
  ;[1 1 1] Tether Logic
  (bind _btethered (CallFunction|AttemptTether TetherActor))
  (if _btethered
    (CallFunction|ForceKill)
    (else
      (CallFunction|ClientReelIn))))

(event Custom|ClientReelIn
  ;[1 1 1] Tether Logic
  (CallFunction|ReelIn))

(event OnComponentBeginOverlap(CapsuleComponent)
  (params (OverlappedComponent :type Primitive%20Component%20Object%20Reference) (OtherActor :type Actor%20Object%20Reference) (OtherComp :type Primitive%20Component%20Object%20Reference) (OtherBodyIndex :type Integer) (bFromSweep :type Boolean) (SweepResult :type Hit%20Result%20Structure%20%28by%20ref%29))
  (bind _wire (Variables|Default|GetWire))
  ;[1 1 1] Overlap - If we overlap water turn off collsion, tighen the rope, start the idle bob, and check to see if we are in a fishing spawner.
  (if (not (Variables|Default|GetbDidHitMeatball))
    (bind _btethered (CallFunction|AttemptTether OtherActor))
    (if _btethered
      (CallFunction|ForceKill)
      (else
        (bind _isignored (CallFunction|IgnoreTetherablePawn OtherActor))
        (if (not _isignored)
          (bind _asbga_athena_flopper_spawn_parent (Utilities|Casting|CastToBGA_Athena_FlopperSpawn_Parent OtherActor)
            (:then
              (Variables|WaterAndCatchingFunctionality|SetChanceForBigBob (+ (Class|BGAAthenaFlopperSpawnParent|GetBonusChanceforBigBob _asbga_athena_flopper_spawn_parent) (Variables|WaterAndCatchingFunctionality|GetChanceForBigBob))))
            (:CastFailed
              (bind _asfort_water_body_actor (Utilities|Casting|CastToFortWaterBodyActor OtherActor)
                (:then
                  (Variables|Default|SetWaterBodyPrimaryColor (Variables|/Script/FortniteGame.FortWaterBodyActor|GetWaterBodyPrimaryColor _asfort_water_body_actor)))
                (:CastFailed
                  (Utilities|Casting|CastToFortAthenaWater OtherActor
                    (:then
                      _)
                    (:CastFailed
                      (bind _asfort_pickup (Utilities|Casting|CastToFortPickup OtherActor))
                      (CallFunction|CustomPickupAttach _asfort_pickup))) ; Check Old Water For Creative
                ))))
          (Utilities|FlowControl|DoOnce false
            (:Completed
              (Variables|WaterAndCatchingFunctionality|SetInWater true)
              (Transformation|SetWorldRotation (Variables|Default|GetMesh) "0, 0, 0" false false)
              (Components|Movement|StopMovementImmediately (Variables|Movement|GetProjectileMovementComponent))
              (Components|Activation|Deactivate (Variables|Default|GetRotatingMovement))
              (Collision|SetCollisionResponsetoChannel (Variables|Collision|GetCapsuleComponent) "ECC_Pawn" "ECR_Ignore")
              (Development|PrintString "HitWater")
              (Fortnite|Fishing|FireEventBeginFishingSession (Variables|Default|GetPlayerPawn))
              (Variables|Default|SetbStartedFishingSession true)
              (CallFunction|CheckforDip)
              (Utilities|IsValid _wire
                (:Is%20Valid
                  (Class|BAthenaFloppingRabbitWire|TightenRope _wire))
                (:Is%20Not%20Valid
                  _))
              (Ability|SendGameplayEventtoActor (Variables|Default|GetPlayerPawn) (Variables|Tags|GetEvent_InWorld))
              (CallFunction|HitWater false)
              (CallFunction|RegisterwithGoalManager)
              (if (GameplayTags|HasMatchingGameplayTag (Variables|Default|GetPlayerPawn) "(TagName=\"Gameplay.Flopper.FreeFishing\")")
                ;[1 1 1] Fish anywhere functionality
                (bind _returnvalue (Game|SpawnActorfromClass "/Game/Athena/Items/EnvironmentalItems/FlopperSpawn/BGA_Athena_FlopperSpawn_World.BGA_Athena_FlopperSpawn_World_C" (Math|Transform|MakeTransform (Transformation|GetActorLocation)) "AlwaysSpawn" "OverrideRootScale" (Variables|Actor|GetInstigator) true false 0 (Variables|Actor|GetInstigator)))
                (Variables|/Game/Athena/Items/EnvironmentalItems/FlopperSpawn/BGA_Athena_FlopperSpawn_Parent.BGA_Athena_FlopperSpawn_Parent_C|SetSpawnedFreeFish true _returnvalue)
                (bind _output_get (Variables|InSpawnerCheck|SetSpawner _returnvalue))
                (Variables|InSpawnerCheck|SetSpawnedDefaultSpawner _output_get)
                (else
                  ;[1 1 1] Spawn Default Flopper spawner.
                  (bind _returnvalue_2 (Game|SpawnActorfromClass "/Game/Athena/Items/EnvironmentalItems/FlopperSpawn/BGA_Athena_FlopperSpawn_Default.BGA_Athena_FlopperSpawn_Default_C" (Math|Transform|MakeTransform (Transformation|GetActorLocation)) "AlwaysSpawn" "OverrideRootScale" (Variables|Actor|GetInstigator) false 0 (Variables|Actor|GetInstigator)))
                  (Variables|InSpawnerCheck|SetSpawner _returnvalue_2)
                  (Variables|InSpawnerCheck|SetSpawnedDefaultSpawner _returnvalue_2))))))))))

(event Custom|Minigame_MoveBobber
  ;[1 1 1] Move bobber back and forth during mini game
  (Utilities|FlowControl|Sequence
    (:then_0
      (Variables|Default|SetTowardPlayerVector (Math|Vector|Normalize (- (Transformation|GetActorLocation (Variables|Default|GetPlayerPawn)) (Transformation|GetActorLocation))))
      (bind _isa (Utilities|FlowControl|FlipFlop)
        (:A
          _)
        (:B
          _))
      (Variables|Default|SetTarget (+ (* (Math|Vector|CrossProduct (Variables|Default|GetTowardPlayerVector) "0, 0,1.000000") (Math|Float|SelectFloat (Math|Random|RandomFloatinRange -5e+02 -2e+02) (Math|Random|RandomFloatinRange 2e+02 5e+02) _isa)) (Transformation|GetActorLocation))))
    (:then_1
      (Timeline|Timeline_0 PlayFromStart
        (:Update
          (Transformation|SetActorLocation (Math|Interpolation|VInterpTo (Transformation|GetActorLocation) (Variables|Default|GetTarget) (stale EventTick DeltaSeconds) (Math|Random|RandomFloatinRange 0.5 1.5))))))))

(event EventTick
  (params (DeltaSeconds :type Float%20%28single-precision%29)))

(fn AttemptTether
  (params (HitActor :type Actor%20Object%20Reference))
  (bind _playerpawn (Variables|Default|GetPlayerPawn))
  (bind _tetherhitactorpawn (LocalVariables|GetTetherHitActorPawn))
  (bind _capsulecomponent (Variables|Collision|GetCapsuleComponent))
  (Networking|SwitchHasAuthority
    (:Authority
      (bind _asfort_athena_vehicle (Utilities|Casting|CastToFortAthenaVehicle HitActor)
        (:then
          (if (GameplayTags|HasTag (Class|FortVehicleInterface|GetVehicleTags _asfort_athena_vehicle) "(TagName=\"Vehicle.Meatball\")")
            (LocalVariables|SetTetherHitActorPawn _asfort_athena_vehicle)
            (Variables|Default|SetbDidHitMeatball true)))
        (:CastFailed
          (bind _asfort_aipawn (Utilities|Casting|CastToFortAIPawn HitActor)
            (:then
              (LocalVariables|SetTetherHitActorPawn _asfort_aipawn)
              (if (GameplayTags|HasTag (GameplayTags|GetOwnedGameplayTags _asfort_aipawn :!class /Script/GameplayTags.GameplayTagAssetInterface) "(TagName=\"NPC.Tetherable\")")
                (Tether|SetTetherPawn _playerpawn _tetherhitactorpawn :!label _label)
                (LocalVariables|SetReturnTethered (Utilities|Equal(Object) (Class|FortPlayerPawn|GetTetherPawn _playerpawn) _tetherhitactorpawn))
                (Transformation|AttachComponentToComponent (Variables|Collision|GetCapsuleComponent) (Variables|Transformation|GetRootComponent (LocalVariables|GetTetherHitActorPawn)) "None" "SnapToTarget" "SnapToTarget" "KeepRelative" true)
                (Rendering|SetActorHiddenInGame true)
                (Game|Components|ProjectileMovement|StopSimulating (Variables|Movement|GetProjectileMovementComponent) (Collision|MakeHitResult true))
                (Components|Activation|Deactivate (Variables|Default|GetRotatingMovement))
                (Collision|SetCollisionResponsetoChannel _capsulecomponent "ECC_Pawn" "ECR_Ignore")
                (Collision|SetCollisionEnabled _capsulecomponent "NoCollision")
                (Collision|SetActorEnableCollision false)
                (Ability|SendGameplayEventtoActor (Variables|GetHitActor) (Variables|Tags|GetEvent_Hooked) (Utilities|Struct|MakeGameplayEventData (Variables|Tags|GetEvent_Hooked) (Variables|Default|GetPlayerPawn)) :!pinned true) ; HOOKED!\n\nLet the TetherPawn know it has been hooked.
                (return (LocalVariables|GetReturnTethered))
                (else
                  _)))
            (:CastFailed
              _))
          (return false)))
      (goto _label))
    (:Remote
      _))
  (return false))

(fn DetermineFishLootTierNameAndSetLootForSpawner
  (params (bIsExplosion :type Boolean))
  (bind _asbga_athena_flopper_spawn_parent (Utilities|Casting|CastToBGA_Athena_FlopperSpawn_Parent (Variables|InSpawnerCheck|GetSpawner))
    (:then
      (bind _output_get (Class|BGAAthenaFlopperSpawnParent|GetLootTier _asbga_athena_flopper_spawn_parent bIsExplosion false (Variables|Default|GetPlayerPawn)))
      (return _output_get))
    (:CastFailed
      (return "None"))))

(fn GetSheetVals
  (bind _output_get (Variables|WaterAndCatchingFunctionality|SetChanceForBigBob (ScalableFloat|GetValueatLevel (Variables|Rows|GetRow_ChanceForBigBob))))
  (Variables|WaterAndCatchingFunctionality|SetStartingCatchBigBobChance _output_get)
  (Variables|WaterAndCatchingFunctionality|SetIncreaseBigBobChanceAmount (ScalableFloat|GetValueatLevel (Variables|Rows|GetRow_ChanceForBigBobIncrease)))
  (Variables|WaterAndCatchingFunctionality|SetDelay_CatchWindow (ScalableFloat|GetValueatLevel (Variables|Rows|GetRow_CatchWindow))))

(fn HandleNothingCaughtFishingAnalytics
  (bind _playerpawn (Variables|Default|GetPlayerPawn))
  (bind (_outactors _returnvalue) (Collision|CapsuleOverlapActors (Transformation|GetActorLocation) 7e+01 1e+03 (Variables|InSpawnerCheck|GetObjectTypes) "/Game/Athena/Items/EnvironmentalItems/FlopperSpawn/BGA_Athena_FlopperSpawn_World.BGA_Athena_FlopperSpawn_World_C"))
  (Utilities|FlowControl|Sequence
    (:then_0
      (if (not (Utilities|Array|IsEmpty _outactors))
        (bind _asbga_athena_flopper_spawn_world (Utilities|Casting|CastToBGA_Athena_FlopperSpawn_World (Utilities|Array|Get(acopy) _outactors)))
        (LocalVariables|SetTempFishPoolTag (Variables|/Game/Athena/Items/EnvironmentalItems/FlopperSpawn/BGA_Athena_FlopperSpawn_Parent.BGA_Athena_FlopperSpawn_Parent_C|GetFishPoolTag _asbga_athena_flopper_spawn_world))))
    (:then_1
      (Utilities|IsValid _playerpawn
        (:Is%20Valid
          (if (Variables|Default|GetbStartedFishingSession)
            (Fortnite|Fishing|FireEventEndFishingSessionwithData _playerpawn (Utilities|Struct|MakePlayerFishingTelemetryData 0 (Variables|/Script/FortniteGame.FortPawn|GetCurrentWeapon _playerpawn) _returnvalue false "" (LocalVariables|GetTempFishPoolTag)))))))))

(fn IgnoreTetherablePawn
  (params (HitActor :type Actor%20Object%20Reference))
  (bind _asfort_aipawn (Utilities|Casting|CastToFortAIPawn HitActor)
    (:then
      (return (GameplayTags|HasTag (GameplayTags|GetOwnedGameplayTags _asfort_aipawn :!class /Script/GameplayTags.GameplayTagAssetInterface) "(TagName=\"NPC.Tetherable\")")))
    (:CastFailed
      (return false))))

(fn ListenforFishingComplete
  (params (Listen :type Boolean))
  (bind _outputdelegate (EventDispatchers|CreateEvent :!event ReelIn))
  ;[0 0.9267 1 0.2] Listens or stops listening for fishing complete on the current spawner. It uses this to Reel itself in.
  (bind _spawner (Variables|InSpawnerCheck|GetSpawner)
    (:then
      (bind _asbga_athena_flopper_spawn_parent (Utilities|Casting|CastToBGA_Athena_FlopperSpawn_Parent _spawner))
      (if Listen
        (Default|BindEventtoOnFishingComplete _asbga_athena_flopper_spawn_parent _outputdelegate)
        (else
          (Default|UnbindEventfromOnFishingComplete _asbga_athena_flopper_spawn_parent _outputdelegate))))
    (:else)))

(fn RegisterWithGoalManager
  (Networking|SwitchHasAuthority
    (:Authority
      (bind (_assignmentidentifier _creationresult) (AI|GoalSelection|CreateWorldAssignment (Variables|Default|GetAIAssignmentIdentifier) (Variables|Default|GetAIAssignmentSettings))) ; Will do nothing and return AssignmmentNotFoundOrCreated if there is no goal manager
      (switch Utilities|FlowControl|Switch|SwitchonEAssignmentCreationResult _creationresult
        (:AssignmentCreated
          _)
        (:AssignmentFound
          _))
      (bind _goal (AI|GoalSelection|MakeGoalfromActor self))
      (AI|GoalSelection|AddGoaltoWorldAssignment _assignmentidentifier _goal))))

(fn UnRegisterWithGoalManager
  (Networking|SwitchHasAuthority
    (:Authority
      (bind _goal (AI|GoalSelection|MakeGoalfromActor self))
      (AI|GoalSelection|RemoveGoalfromWorldAssignment (Variables|Default|GetAIAssignmentIdentifier) _goal))))

(fn ConstructionScript
  (|CallParentFunction))
